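# Unlit, two-sided translucent material for the Celestial backpack, rendering an animated nebula and inky smoke effect on mesh particles.
# Panning smoke masks on UV1 blend tint colors with a contrast-adjusted galaxy nebula, then desaturation, tint and an emissive adjustment function follow.
# A static switch adds a third color variation.
# UV0 gradients and particle alpha shape opacity, with depth fade and opacity multiplier skipped on forward and mobile shading paths.
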Material /Game/Accessories/FORT_Backpacks/M_MED_Celestial_Backpack/M_Backpack_Celestial
graph {
 properties {
  bUsedWithMeshParticles=true
  bUsedWithNiagaraMeshParticles=true
  blendMode=BLEND_Translucent
  pixelDepthOffsetMode=PDOM_Legacy
  shadingModel=MSM_Unlit
  twoSided=true
  usedShadingModels=Unlit
 }
 nodes [
  n1 = VectorParameter() {parameterName=ColorB, defaultValue=<rgba 0.0114 0 0.026 1>}
  n2 = TextureCoordinate() {coordinateIndex=1, uTiling=0.7, vTiling=0.7}
  n3 = LinearInterpolate(
   A=VectorParameter() {parameterName=ColorC, defaultValue=<rgba 0.0114 0 0.026 1>}.RGB
   B=@n1.RGB
   Alpha=TextureSample(UVs=Panner(Coordinate=@n2) {speedX=-0.045, speedY=0.08}) {texture=&T_Inky_Smoke_Tile}.R
  )
  n4 = Panner(Coordinate=TextureCoordinate() {coordinateIndex=1, uTiling=0.5}) {speedX=-0.045, speedY=0.08}
  n5 = TextureSample(UVs=@n4) {texture=&T_Inky_Smoke_Tile}
  n6 = Power(
   Base=TextureSample(UVs=@n4) {texture=/Game/Characters/Player/Male/Medium/Bodies/M_Med_Soldier_04/Skins/Galaxy/Textures/T_FN_Nebula}.RGB
   Exp=ScalarParameter() {defaultValue=1, parameterName=Noise_Contrast}
  )
  n7 = ScalarParameter() {parameterName=Desaturation_Noise}
  n8 = LinearInterpolate(
   A=LinearInterpolate(
    A=StaticSwitchParameter(True=@n3, False=@n1.RGB) {parameterName=UseColorCforColorVariation}
    B=VectorParameter() {parameterName=ColorA, defaultValue=<rgba 0 0.0411 0.1823 1>}.RGB
    Alpha=@n5.R
   )
   B=Multiply(A=Desaturation(None=@n6, Fraction=@n7), B=ScalarParameter() {defaultValue=1, parameterName=NoiseBrightness})
   Alpha=@n5.R
  )
  n9 = ScalarParameter() {parameterName=Desaturation_Total}
  MaterialFunctionCall(Emissive=Multiply(A=VectorParameter() {parameterName=Color_Total, defaultValue=<rgba 1 1 1 1>}.RGB, B=Desaturation(None=@n8, Fraction=@n9))) {materialFunction=&MF_FX_EmissiveAdjustments}
  n10 = MaterialFunctionCall(Float2=TextureCoordinate()) {materialFunction=&BreakOutFloat2Components}
  n11 = Multiply(A=Sine(None=Multiply(A=@n10.R) {constB=0.5})) {constB=2}
  n12 = Multiply(A=@n10.G) {constB=4}
  n13 = Multiply(A=ParticleColor().A, B=Multiply(A=Saturate(None=@n11), B=Saturate(None=@n12)))
  n14 = Multiply(
   A=DepthFade(Opacity=@n13, FadeDistance=ScalarParameter() {parameterName=FadeDistance}) {fadeDistanceDefault=0}
   B=ScalarParameter() {defaultValue=1, parameterName=OpacityMultiply}
  )
  ShadingPathSwitch(Default=@n14, Deferred=@n14, Forward=@n13, Mobile=@n13)
 ]
}
