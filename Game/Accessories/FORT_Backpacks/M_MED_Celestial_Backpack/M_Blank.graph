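# M_Blank is a placeholder-style translucent, unlit material stored with the Celestial backpack cosmetic.
# Its graph holds only a single Constant at its default value, with no connection to a material output shown.
# There are no textures or parameters.
# Usage flags allow it on skeletal meshes, morph targets, mesh deformers, and mesh particles in both Cascade and Niagara.
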Material /Game/Accessories/FORT_Backpacks/M_MED_Celestial_Backpack/M_Blank
graph {
 properties {
  bUsedWithMeshDeformer=true
  bUsedWithMeshParticles=true
  bUsedWithMorphTargets=true
  bUsedWithNiagaraMeshParticles=true
  bUsedWithSkeletalMesh=true
  blendMode=BLEND_Translucent
  pixelDepthOffsetMode=PDOM_Legacy
  shadingModel=MSM_Unlit
  usedShadingModels=Unlit
 }
 nodes [Constant()]
}
